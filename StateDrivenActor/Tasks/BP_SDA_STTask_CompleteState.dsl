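; BP_SDA_STTask_CompleteState is a StateTree task for the StateDrivenActor system that completes the state it belongs to.
; On entering the state it immediately calls Finish Task with success, without any other logic or use of the transition data.
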(event EventEnterState
  (params (Transition :type State%20Tree%20Transition%20Result%20Structure%20%28by%20ref%29))
  (StateTree|FinishTask true))
